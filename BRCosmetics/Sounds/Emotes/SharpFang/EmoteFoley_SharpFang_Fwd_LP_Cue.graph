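# Looping foley sound cue for the SharpFang emote, apparently for its forward movement.
# Two goop loops are mixed together, then blended with a low loop.
# A continuous modulator at the root scales volume and pitch from the EmoteMoveSpeed parameter, so the sound follows movement speed.
# It uses emote foley attenuation, class and concurrency, keeps playing when silent, and has a send to the music audio-analysis submix.
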SoundCue /BRCosmetics/Sounds/Emotes/SharpFang/EmoteFoley_SharpFang_Fwd_LP_Cue
root=ModulatorContinuous_0
nodes [
 {id=Mixer_0, children [Mixer_1,WavePlayer_1], properties {inputVolume [1,0.8]}}
 {id=Mixer_1, children [WavePlayer_0,WavePlayer_2], properties {inputVolume [1.5,0.8]}}
 {
  id=ModulatorContinuous_0
  children [Mixer_0]
  properties {pitchModulationParams {minOutput=1, parameterName=EmoteMoveSpeed}, volumeModulationParams {parameterName=EmoteMoveSpeed}}
 }
 {id=WavePlayer_0, properties {bLooping=true, soundWaveAssetPtr=&/Emote_SharpFang_Fwd_LP_Goop}}
 {id=WavePlayer_1, properties {bLooping=true, soundWaveAssetPtr=&/Emote_SharpFang_Fwd_LP_Low}}
 {id=WavePlayer_2, properties {bLooping=true, soundWaveAssetPtr=&/Emote_SharpFang_Fwd_LP_Goop_Low}}
]
properties {
 attenuationSettings=&Fort_Emotes_Athena_Att
 audioPropertiesSheet=/BRCosmetics/Sounds/Settings/PropertySheets/Emotes/APS_Cosmetic_Emote_Foley
 maxDistance=1600
 soundClassObject=/Game/Sounds/FortSoundClassesAndModes/Fort_Foley_Emote
 virtualizationMode=PlayWhenSilent
 volumeMultiplier=1
 concurrencyOverrides {maxCount=3}
 concurrencySet [/Game/Sounds/Concurrency_Rules/Athena_EmoteFoley_Conc]
 soundSubmixSends [{minSendLevel=0, soundSubmix=&Fortnite_AudioAnalysis_Music}]
}
